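# Stereo one-shot MetaSound source for outdoor ambient sounds, with attenuation and a send to the exterior ambient source bus.
# On play, it picks a random wave from an input array, with the no-repeat count set to the array length.
# MSP_RandomPitchTimeVolume randomizes trigger delay, wave player pitch and mixer gain within exposed ranges.
# The wave array is empty by default, and the source finishes when the wave player does.
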MetaSoundSource /FortMetasound/Ambient/MSS_AmbientOneShots
inputs [
 Delay.DelayMax {type=Time}
 Delay.DelayMin {type=Time}
 Pitch.PitchMax {type=Float}
 Pitch.PitchMin {type=Float}
 UE.Source.OnPlay {type=Trigger}
 Volume.VolumeMax {type=Float}
 Volume.VolumeMin {type=Float}
 Wave.Wave Assets {type=WaveAsset:Array}
]
interfaces [UE.OutputFormat.Stereo:1.0, UE.Source.OneShot:1.0, UE.Source:1.1]
outputs [
 "UE.OutputFormat.Stereo.Audio:0" {type=Audio}
 "UE.OutputFormat.Stereo.Audio:1" {type=Audio}
 UE.Source.OneShot.OnFinished {type=Trigger}
]
pages [Default {
  input_defaults {
   Delay.DelayMax=0
   Delay.DelayMin=0
   Pitch.PitchMax=0.8
   Pitch.PitchMin=-1
   UE.Source.OnPlay=null
   Volume.VolumeMax=3
   Volume.VolumeMin=-0.6
   Wave.Wave Assets []
  }
  nodes [
   n1 = InitVariable.Float()
   ArrayRandomGet_0 = Array.Random Get:WaveAsset:Array(Next=^UE.Source.OnPlay, In Array=^Wave.Wave Assets, No Repeats=Array.Num:WaveAsset:Array(Array=^Wave.Wave Assets).Num)
   n2 = /FortMetasound/Patches/MSP_RandomPitchTimeVolume(
    "Delay Max (Seconds)"=^Delay.DelayMax
    "Delay Min (Seconds)"=^Delay.DelayMin
    "Random Pitch Max (Semitones)"=^Pitch.PitchMax
    "Random Pitch Min (Semitones)"=^Pitch.PitchMin
    "Random Volume Change Max (Decibels)"=^Volume.VolumeMax
    "Random Volume Change Min (Decibels)"=^Volume.VolumeMin
    Trigger=UE.Reroute(Value=@ArrayRandomGet_0.On Next).Value
   )
   n3 = InitVariable.Time()
   TriggerDelay_0 = UE.Trigger Delay(
    In=@ArrayRandomGet_0.On Next
    Delay Time=VariableAccessor.Time(Variable=VariableMutator.Time(Variable=@n3.Variable, Value=@n2.Random Time).Variable).Value
   )
   n4 = InitVariable.Float()
   n5 = InitVariable.WaveAsset()
   WavePlayer_0 = UE.Wave Player:Stereo(
    Play=@TriggerDelay_0.out
    Pitch Shift=VariableAccessor.Float(Variable=VariableMutator.Float(Variable=@n4.Variable, Value=@n2.Random Pitch).Variable).Value
    Wave Asset=VariableAccessor.WaveAsset(Variable=VariableMutator.WaveAsset(Value=@ArrayRandomGet_0.Value, Variable=@n5.Variable).Variable).Value
   ) {also {version [1,1]}}
   AudioMixerAudioMixerStereo2_0 = "AudioMixer.Audio Mixer (Stereo, 2)"(
    Gain 0=VariableAccessor.Float(Variable=VariableMutator.Float(Variable=@n1.Variable, Value=@n2.Random Volume).Variable).Value
    In 0 L=@WavePlayer_0.Out Left
    In 0 R=@WavePlayer_0.Out Right
   )
   InitVariable.Float()
   InitVariable.Float()
   InitVariable.Trigger()
  ]
  links [
   {dst="@UE.OutputFormat.Stereo.Audio:0", src=AudioMixerAudioMixerStereo2_0.Out L}
   {dst="@UE.OutputFormat.Stereo.Audio:1", src=AudioMixerAudioMixerStereo2_0.Out R}
   {dst="@UE.Source.OneShot.OnFinished", src=WavePlayer_0.On Finished}
  ]
 }]
properties {
 assetImportData=&self:AssetImportData
 attenuationSettings=/Game/Sounds/Attenuation/Fort_TileAmbient_OneShot_Wider_Att
 maxDistance=8000
 numChannels=2
 outputFormat=Stereo
 sampleRate=48000
 soundClassObject=&Fort_Ambient_Outdoors_Athena
 volume=0.501
 preEffectBusSends [{
   audioBus=null
   maxSendDistance=1000
   minSendDistance=100
   minSendLevel=0
   sendLevel=1
   soundSourceBus=&AmbientAudio_Exterior_SourceBus
  }]
}
