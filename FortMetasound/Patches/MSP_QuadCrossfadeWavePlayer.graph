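# MetaSound patch that plays a quad wave asset through two wave players and crossfades between them.
# A trigger with a valid wave asset starts MSP_Fade, whose output trigger is routed by the InvertOutput bool to play one of the two players.
# The fade value drives MSP_QuadCrossfade to blend both players into four audio outputs.
# Both players share asset, loop, pitch and stop inputs, and their play and finish triggers are merged.
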MetaSoundPatch /FortMetasound/Patches/MSP_QuadCrossfadeWavePlayer
inputs [
 Fade.CurveExp {type=Float}
 Fade.FadeTime {type=Float}
 Fade.InvertOutput {type=Bool}
 In {type=Trigger}
 Loop {type=Bool}
 Loop Duration {type=Time}
 Loop Start {type=Time}
 Pitch Shift {type=Float}
 Sound.Sound {type=WaveAsset}
 Start Time {type=Time}
 Stop {type=Trigger}
]
outputs [
 Any Finished {type=Trigger}
 Any On Play {type=Trigger}
 Fade.OnDone {type=Trigger}
 FrontLeft {type=Audio}
 FrontRight {type=Audio}
 OutSideLeft {type=Audio}
 OutSideRight {type=Audio}
]
pages [Default {
  input_defaults {
   Fade.CurveExp=1
   Fade.FadeTime=1
   Fade.InvertOutput=null
   In=null
   Loop="<MetaSoundBuilderResult.SUCCEEDED: 0>"
   Loop Duration=-1
   Loop Start=0
   Pitch Shift=0
   Sound.Sound=null
   Start Time=0
   Stop=null
  }
  nodes [
   MSP_Fade_0 = &MSP_Fade(
    Curve Exp=^Fade.CurveExp
    Fade Time=^Fade.FadeTime
    bInvertOutput=^Fade.InvertOutput
    In=&/IsWaveAssetValid(Trigger In=^In, WaveAsset=^Sound.Sound).Is Valid
   )
   n1 = TriggerCompare.Clamp:Bool(A=^Fade.InvertOutput, Compare=@MSP_Fade_0.Out) {also {defaults {B="<MetaSoundBuilderResult.SUCCEEDED: 0>"}}}
   WavePlayer_0 = UE.Wave Player:Quad(
    Loop=^Loop
    Loop Duration=^Loop Duration
    Loop Start=^Loop Start
    Pitch Shift=^Pitch Shift
    Wave Asset=^Sound.Sound
    Start Time=^Start Time
    Stop=^Stop
    Play=UE.Reroute(Value=@n1.TRUE).Value
   ) {also {version [1,1]}}
   WavePlayer_1 = UE.Wave Player:Quad(
    Loop=^Loop
    Loop Duration=^Loop Duration
    Loop Start=^Loop Start
    Pitch Shift=^Pitch Shift
    Wave Asset=^Sound.Sound
    Start Time=^Start Time
    Stop=^Stop
    Play=UE.Reroute(Value=@n1.FALSE).Value
   ) {also {version [1,1]}}
   MSP_QuadCrossfade_0 = &/MSP_QuadCrossfade(
    FloatIn=UE.Reroute(Value=UE.Reroute(Value=@MSP_Fade_0.Value Out).Value).Value
    Input 1 FL=@WavePlayer_0.Out Front Left
    Input 1 FR=@WavePlayer_0.Out Front Right
    Input 1 SL=@WavePlayer_0.Out Side Left
    Input 1 SR=@WavePlayer_0.Out Side Right
    Input 2 FL=@WavePlayer_1.Out Front Left
    Input 2 FR=@WavePlayer_1.Out Front Right
    Input 2 SL=@WavePlayer_1.Out Side Left
    Input 2 SR=@WavePlayer_1.Out Side Right
   )
   TriggerAnyTriggerAny2_0 = "TriggerAny.Trigger Any (2)"(In 0=@WavePlayer_0.On Finished, In 1=@WavePlayer_1.On Finished)
   TriggerAnyTriggerAny2_1 = "TriggerAny.Trigger Any (2)"(In 0=@WavePlayer_0.On Play, In 1=@WavePlayer_1.On Play)
  ]
  links [
   {dst="@Fade.OnDone", src=MSP_Fade_0.On Done}
   {dst="@FrontLeft", src=MSP_QuadCrossfade_0.FrontLeft}
   {dst="@FrontRight", src=MSP_QuadCrossfade_0.FrontRight}
   {dst="@OutSideLeft", src=MSP_QuadCrossfade_0.OutSideLeft}
   {dst="@OutSideRight", src=MSP_QuadCrossfade_0.OutSideRight}
   {dst="@Any Finished", src=TriggerAnyTriggerAny2_0.out}
   {dst="@Any On Play", src=TriggerAnyTriggerAny2_1.out}
  ]
 }]
